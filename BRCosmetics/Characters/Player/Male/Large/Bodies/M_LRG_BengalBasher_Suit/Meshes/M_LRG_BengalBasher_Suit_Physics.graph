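# Physics asset for the Bengal Basher suit, a large male player body cosmetic, covering its dynamic knife, star and strap bones.
# Boxes and capsules wrap those accessories, while pelvis and spine carry capsule and sphere shapes for the torso.
# Constraints with small angular limits tie the knives to pelvis or spine_01 and link the strap to the star.
# The second left knife has its twist locked, and no constraint links the star-strap pair to the torso.
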PhysicsAsset /BRCosmetics/Characters/Player/Male/Large/Bodies/M_LRG_BengalBasher_Suit/Meshes/M_LRG_BengalBasher_Suit_Physics
bodies {
 dyn_knife_l_1_joint [{
   shapeType=Box
   extentX=15.5
   extentY=1.5
   extentZ=7.5
   shapeName=dyn_knife_l_1_joint_box
   center=<-2.3772 0.3654 -1.6262>
   rotation=<rot -0.0003 -6.2422 0.0002>
  }]
 dyn_knife_l_2_joint [{
   shapeType=Capsule
   length=15.5
   radius=1.5
   shapeName=dyn_knife_l_2_joint_capsule
   center=<0 0 0.6959>
   rotation=<rot 84.9996 -4.9997 -0.0004>
  }]
 dyn_knife_m_joint [{shapeType=Capsule, length=28.5, radius=2.5, shapeName=dyn_knife_m_joint_capsule}]
 dyn_knife_r_1_joint [{
   shapeType=Box
   extentX=15.5
   extentY=1.5
   extentZ=7.5
   shapeName=dyn_knife_l_1_joint_box
   center=<-2.374 0.5174 1.6341>
   rotation=<rot -0.0425 -5.4193 -177.036>
  }]
 dyn_knife_r_2_joint [{
   shapeType=Capsule
   length=16
   radius=1.5
   shapeName=dyn_knife_r_2_joint_capsule
   center=<0.0263 0.3003 -0.5927>
   rotation=<rot -90 0 -4.9995>
  }]
 dyn_star_joint [{shapeType=Capsule, length=2.5, radius=1.5, shapeName=dyn_star_joint_capsule, rotation=<rot -90 90 -90.0004>}]
 dyn_strap_1_joint [{shapeType=Capsule, length=1.5, radius=1.5, shapeName=dyn_strap_1_joint_capsule, center=<0.5722 0 0>, rotation=<rot -90 -179.9999 -179.9999>}]
 pelvis [{shapeType=Capsule, length=5, radius=14.4151, center=<-4.8636 0.2289 0.4113>, rotation=<rot -0.4754 -90.0015 -89.908>}]
 spine_01 [{shapeType=Capsule, length=11.5, radius=11.5, shapeName=spine_01_capsule, center=<1.3779 1.4638 0.2583>}]
 spine_02 [{shapeType=Sphere, radius=17.5, shapeName=spine_02_sphere}]
}
constraints [
 {bone1Name=dyn_knife_m_joint, bone2Name=pelvis, swing1LimitDegrees=1, swing2LimitDegrees=1, twistLimitDegrees=1}
 {bone1Name=dyn_knife_l_2_joint, bone2Name=spine_01, swing1LimitDegrees=5, swing2LimitDegrees=5, twistLimitDegrees=1, twistMotion=Locked}
 {bone1Name=dyn_knife_r_2_joint, bone2Name=spine_01, swing1LimitDegrees=5, swing2LimitDegrees=5, twistLimitDegrees=1}
 {bone1Name=dyn_strap_1_joint, bone2Name=dyn_star_joint, swing1LimitDegrees=10, swing2LimitDegrees=20, twistLimitDegrees=10}
 {bone1Name=dyn_knife_l_1_joint, bone2Name=pelvis, swing1LimitDegrees=10, swing2LimitDegrees=10, twistLimitDegrees=5}
 {bone1Name=dyn_knife_r_1_joint, bone2Name=pelvis, swing1LimitDegrees=5, swing2LimitDegrees=10, twistLimitDegrees=5}
]
