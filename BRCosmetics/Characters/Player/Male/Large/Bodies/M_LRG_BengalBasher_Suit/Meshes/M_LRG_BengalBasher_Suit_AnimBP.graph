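# Animation Blueprint for the male large BengalBasher suit cosmetic body.
# It copies the parent pose through Control Rig and rigid-body physics, then layers additive knife-bone offsets when crouching or crouch-moving and middle-finger corrections driven by joint bend.
# Each update feeds leg motion into two material WPO cloth regions and enables wind while skydive floating.
# On customization load, it caches dynamic materials and hides one mesh part under cape backblings.
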AnimBlueprint /BRCosmetics/Characters/Player/Male/Large/Bodies/M_LRG_BengalBasher_Suit/Meshes/M_LRG_BengalBasher_Suit_AnimBP
target_skeleton=&/M_LRG_BengalBasher_Suit_Skeleton
graphs [
 AnimGraph {nodes [
   n1 = Control Rig(Source=Copy Pose From Mesh() {bUseAttachedParent=true}) {
    alphaBoolBlend {blendOption=Linear}
    controlRigAssetReference {blueprintRigClass=&/M_LRG_BengalBasher_Suit_CtrlRig.M_LRG_BengalBasher_Suit_CtrlRig_C}
   }
   n2 = RigidBody(ComponentPose=Local To Component(LocalPose=@n1)) {
    overridePhysicsAsset=&/M_LRG_BengalBasher_Suit_Physics
    simulationSpace=BaseBoneSpace
    alphaBoolBlend {blendOption=Linear}
    baseBoneRef {boneName=spine_01}
    componentAppliedLinearAccClamp=<3000 3000 3000>
    componentLinearAccScale=<0.4 0.4 0.4>
    componentLinearVelScale=<0.4 0.4 0.4>
   }
   n3 = Get bIsCrouching()
   n4 = "Transform (Modify) Bone - Bone: dyn_knife_r_2_joint"!(
    ComponentPose=Local To Component(LocalPose=Component To Local(ComponentPose=@n2))
    bAlphaBoolEnabled:bool=@n3
    Translation="0,0,-2"
    Rotation="0,0,30"
    Scale="1,1,1"
   ) -> (Pose) {
    alphaInputType=Bool
    rotationMode=BMM_Additive
    rotationSpace=BCS_BoneSpace
    translationMode=BMM_Additive
    translationSpace=BCS_BoneSpace
    alphaBoolBlend {blendInTime=0.3, blendOption=Linear, blendOutTime=0.3}
    boneToModify {boneName=dyn_knife_r_2_joint}
    rotation {roll=30}
    translation {z=-2}
   }
   n5 = "Transform (Modify) Bone - Bone: dyn_knife_l_2_joint"!(ComponentPose=@n4, bAlphaBoolEnabled:bool=@n3, Translation="0,0.5,5", Rotation="0,0,-30", Scale="1,1,1") -> (Pose) {
    alphaInputType=Bool
    rotationMode=BMM_Additive
    rotationSpace=BCS_BoneSpace
    translationMode=BMM_Additive
    translationSpace=BCS_BoneSpace
    alphaBoolBlend {blendInTime=0.3, blendOption=Linear, blendOutTime=0.3}
    boneToModify {boneName=dyn_knife_l_2_joint}
    rotation {roll=-30}
    translation {y=0.5, z=5}
   }
   n6 = Get crouch move() -> (crouch move:bool)
   n7 = "Transform (Modify) Bone - Bone: dyn_knife_l_2_joint"(ComponentPose=@n5, bAlphaBoolEnabled=@n6, Translation="0,0,4", Rotation="-0.000,-10.000,0.000", Scale="1,1,1") {
    alphaInputType=Bool
    rotationMode=BMM_Additive
    rotationSpace=BCS_BoneSpace
    translationMode=BMM_Additive
    translationSpace=BCS_BoneSpace
    alphaBoolBlend {blendInTime=0.3, blendOption=Linear, blendOutTime=0.3}
    boneToModify {boneName=dyn_knife_l_2_joint}
    rotation=<rot -0.0001 -10 0>
    translation {z=4}
   }
   n8 = "Transform (Modify) Bone - Bone: dyn_knife_r_2_joint"(ComponentPose=@n7, bAlphaBoolEnabled=@n6, Translation="0.5,0,-4", Rotation="0,0,0", Scale="1,1,1") {
    alphaInputType=Bool
    rotationMode=BMM_Additive
    rotationSpace=BCS_BoneSpace
    translationMode=BMM_Additive
    translationSpace=BCS_BoneSpace
    alphaBoolBlend {blendInTime=0.3, blendOption=Linear, blendOutTime=0.3}
    boneToModify {boneName=dyn_knife_r_2_joint}
    translation {x=0.5, z=-4}
   }
   n9 = Get bIsCrouching()
   n10 = "Transform (Modify) Bone - Bone: dyn_knife_l_1_joint"!(ComponentPose=@n8, bAlphaBoolEnabled:bool=@n9, Translation="0,0,0", Rotation="-70,0,0", Scale="1,1,1") -> (Pose) {
    alphaInputType=Bool
    rotationMode=BMM_Additive
    rotationSpace=BCS_BoneSpace
    translationMode=BMM_Additive
    translationSpace=BCS_BoneSpace
    alphaBoolBlend {blendInTime=0.3, blendOption=Linear, blendOutTime=0.3}
    boneToModify {boneName=dyn_knife_l_1_joint}
    rotation {pitch=-70}
   }
   n11 = "Transform (Modify) Bone - Bone: dyn_knife_r_1_joint"!(ComponentPose=@n10, bAlphaBoolEnabled:bool=@n9, Translation="0,0,1.5", Rotation="50,0,0", Scale="1,1,1") -> (Pose) {
    alphaInputType=Bool
    rotationMode=BMM_Additive
    rotationSpace=BCS_BoneSpace
    translationMode=BMM_Additive
    translationSpace=BCS_BoneSpace
    alphaBoolBlend {blendInTime=0.3, blendOption=Linear, blendOutTime=0.3}
    boneToModify {boneName=dyn_knife_r_1_joint}
    rotation {pitch=50}
    translation {z=1.5}
   }
   n12 = Get crouch move()
   n13 = "Transform (Modify) Bone - Bone: dyn_knife_l_1_joint"(ComponentPose=@n11, bAlphaBoolEnabled=@n12, Translation="-1,0.3,-0.5", Rotation="0,0,0", Scale="1,1,1") {
    alphaInputType=Bool
    rotationMode=BMM_Additive
    rotationSpace=BCS_BoneSpace
    translationMode=BMM_Additive
    translationSpace=BCS_BoneSpace
    alphaBoolBlend {blendInTime=0.3, blendOption=Linear, blendOutTime=0.3}
    boneToModify {boneName=dyn_knife_l_1_joint}
    translation=<-1 0.3 -0.5>
   }
   n14 = "Transform (Modify) Bone - Bone: dyn_knife_r_1_joint"(ComponentPose=@n13, bAlphaBoolEnabled=@n12, Translation="2,0,2", Rotation="0,-10,10", Scale="1,1,1") {
    alphaInputType=Bool
    rotationMode=BMM_Additive
    rotationSpace=BCS_BoneSpace
    translationMode=BMM_Additive
    translationSpace=BCS_BoneSpace
    alphaBoolBlend {blendInTime=0.3, blendOption=Linear, blendOutTime=0.3}
    boneToModify {boneName=dyn_knife_r_1_joint}
    rotation {roll=10, yaw=-10}
    translation {x=2, z=2}
   }
   n15 = "Transform (Modify) Bone - Bone: middle_01_l"(ComponentPose=@n14, Alpha=Get Transform L() -> (Transform L:float), Translation="0,0,0", Rotation="-10,0,10", Scale="1,1,1") {
    rotationMode=BMM_Additive
    rotationSpace=BCS_BoneSpace
    alphaBoolBlend {blendOption=Linear}
    boneToModify {boneName=middle_01_l}
    rotation {pitch=-10, roll=10}
   }
   n16 = "Transform (Modify) Bone - Bone: middle_01_r"(ComponentPose=@n15, Alpha=Get Transform R() -> (Transform R:float), Translation="0,0,0", Rotation="-10,0,10", Scale="1,1,1") {
    rotationMode=BMM_Additive
    rotationSpace=BCS_BoneSpace
    alphaBoolBlend {blendOption=Linear}
    boneToModify {boneName=middle_01_r}
    rotation {pitch=-10, roll=10}
   }
   Output Pose(Result=Component To Local(ComponentPose=@n16)) {layerGroup=None}
  ]}
 EventGraph {nodes [
   n1 = Event BlueprintUpdateAnimation()
   > Set WPO Globals Function(
    Material=Get Material Instance() -> (Material Instance:Material Instance Dynamic Object Reference)
    Overlay Material=Get Mid Overlay() -> (Mid Overlay:Material Instance Dynamic Object Reference)
    Skeletal Mesh=Get SkelMesh() -> (SkelMesh:Skeletal Mesh Component Budgeted Object Reference)
    "Local Noise Transform  Location"="Get Local Noise Transform  Location"() -> ("Local Noise Transform  Location":Vector)
    Local Noise Offset=Get Local Noise Offset()
   )
   > Set WPO Regional Function!(
    execute
    Material:Material Instance Dynamic Object Reference=Get Material Instance()
    Overlay Material:Material Instance Dynamic Object Reference=Get Mid Overlay()
    Skeletal Mesh:Skeletal Mesh Component Object Reference=Get SkelMesh()
    Velocity Transform:Name=foot_r
    Velocity Interp Speed:double=Get Velocity Interp Speed()
    Collision Transform:Name=calf_r
    Collision Transform Axis:EAxis Enum
    Invert Collision Transform Axis:bool=false
    Delta Time X:double=@n1.DeltaTimeX
    Previous Velocity Transform Location:"Vector (by ref)"=Get R1 Velocity Transform Location()
    Previous Velocity Transform Velocity:"Vector (by ref)"=Get R1 Velocity Transform Velocity()
    Region:int=1
   ) -> (then)
   > Set WPO Regional Function(
    Material=Get Material Instance()
    Overlay Material=Get Mid Overlay()
    Skeletal Mesh=Get SkelMesh()
    Velocity Transform=foot_l
    Velocity Interp Speed=Get Velocity Interp Speed()
    Collision Transform=calf_l
    Invert Collision Transform Axis=true
    Delta Time X=@n1.DeltaTimeX
    Previous Velocity Transform Location=Get R2 Velocity Transform Location()
    Previous Velocity Transform Velocity=Get R2 Velocity Transform Velocity()
    Region=2
   )
   > n2 = Branch(Condition=Get bIsSkydiveFloating())
   > n3 = Set WPO Pho Wind Intensity(WPO Pho Wind Intensity:float=1) -> (Output_Get:float)
   SetScalarParameterValue(
    execute=[@n4.then, @n3.then]
    self=[Get Material Instance(), Get Mid Overlay()]
    ParameterName=ClothSim_WindStrength
    Value=Get WPO Pho Wind Intensity() -> (WPO Pho Wind Intensity:float)
   )
   > Set crouch move(crouch move:bool=AND Boolean(A=Get bIsCrouching(), B=Get bIsMoving2D())) -> (Output_Get:bool)
   > Set Transform L(Transform L:float=GetAlphaFromReferenceTransformDelta(
     SkeletalMeshComponent=GetOwnerSkeletalMeshComponent()
     BoneName=middle_02_l
     BaseName=middle_metacarpal_l
     InputRange=(X=-85,Y=-120)
     TwistAxis="1,0,0"
    )) -> (Output_Get:float)
   > Set Transform R(Transform R:float=GetAlphaFromReferenceTransformDelta(
     SkeletalMeshComponent=GetOwnerSkeletalMeshComponent()
     BoneName=middle_02_r
     BaseName=middle_metacarpal_r
     InputRange=(X=-85,Y=-120)
     TwistAxis="1,0,0"
    )) -> (Output_Get:float)
   n4 = Set WPO Pho Wind Intensity(execute=@n2.else, WPO Pho Wind Intensity=0)
   n5 = Event OnCharacterCustomizationLoaded()
   > n6 = Branch(Condition=CheckForBackbling().bIsBackblingACape)
   n7 = "Get (a copy)"(Array:Array of Skeletal Mesh Component Budgeted Object References=@n5.CustomMeshParts, Dimension 1=4) -> (Output:Skeletal Mesh Component Budgeted Object Reference)
   > n8 = SetVisibility!(execute, self:Scene Component Object Reference=@n7, bNewVisibility:bool=false, bPropagateToChildren:bool=false) -> (then)
   n9 = Set SkelMesh(
    execute=[@n8.then, @n12.then]
    SkelMesh:Skeletal Mesh Component Budgeted Object Reference="Get (a copy)"(Array=@n5.CustomMeshParts, Dimension 1=1)
   ) -> (Output_Get:Skeletal Mesh Component Budgeted Object Reference)
   > n10 = Get Dynamic Material Instance(SkeletalMeshComponent=@n9)
   > Set Material Instance(Material Instance:Material Instance Dynamic Object Reference=@n10) -> (Output_Get:Material Instance Dynamic Object Reference)
   > n11 = GetOrCreateDynamicOverlayMaterial(MeshComponent=@n9, MaterialName=None)
   > Set Mid Overlay(Mid Overlay:Material Instance Dynamic Object Reference=@n11) -> (Output_Get:Material Instance Dynamic Object Reference)
   n12 = SetVisibility(execute=@n6.else, self=@n7, bNewVisibility=true, bPropagateToChildren=false)
   Reroute Node(InputPin:Wildcard) -> (OutputPin:Wildcard)
  ]}
]
variables [
 crouch move {type=boolean}
 Local Noise Offset {type=Vector}
 "Local Noise Transform  Location" {type=Vector}
 Material Instance {type=MaterialInstanceDynamic}
 Mid Overlay {type=MaterialInstanceDynamic}
 R1 Velocity Transform Location {type=Vector}
 R1 Velocity Transform Velocity {type=Vector}
 R2 Velocity Transform Location {type=Vector}
 R2 Velocity Transform Velocity {type=Vector}
 SkelMesh {type=SkeletalMeshComponentBudgeted}
 Transform L {type=number}
 Transform R {type=number}
 Velocity Interp Speed {type=number, default=1}
 WPO Pho Wind Intensity {type=number}
]
